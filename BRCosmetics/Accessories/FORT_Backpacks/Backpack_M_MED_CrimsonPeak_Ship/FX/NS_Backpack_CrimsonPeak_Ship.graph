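# Niagara effect for the CrimsonPeak Ship backpack cosmetic, adding flickering sprite and light glows to the ship's body and ends.
# Two looping burst emitters track the skeletal mesh: one oriented body glow, plus four end glows partitioned by size and offset, with lights positioned for one pair.
# User.Damaged drives color curves from cyan to red, and front-end modifiers adjust alpha between gameplay and menus.
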NiagaraSystem /BRCosmetics/Accessories/FORT_Backpacks/Backpack_M_MED_CrimsonPeak_Ship/FX/NS_Backpack_CrimsonPeak_Ship
dependencies {
 usedDataInterfaces [
  &NiagaraDataInterfaceColorCurve
  &NiagaraDataInterfaceArrayFloat3
  &NiagaraDataInterfaceCurve
  &NiagaraDataInterfaceArrayFloat2
 ]
 usedDynamicInputs [&Vector2DFromFloat]
 usedModules [
  &SystemState
  &self:SystemSpawnScript.SystemScriptSource.SystemScriptGraph.NiagaraNodeAssignment_0.SetVariables_#1
 ]
}
emitters {
 ShipBodyGlow {
  emitterUpdateScript [
   &EmitterState()
   &SpawnBurst_Instantaneous(Use Spawn Probability=false, UseLoopCountLimit=true)
  ]
  particleSpawnScript [
   &InitializeParticle(
    Color=<rgba 1 1 1 0.05>
    Lifetime=2
    Sprite Size=<22 40>
    Sprite Size Mode=::NewEnumerator4
    UsePositionOffset=false
    Write Material Random=true
   )
   &CameraOffset(Camera Offset Amount=10)
  ]
  particleUpdateScript [
   &ParticleState(
    Kill Particles When Lifetime Has Elapsed=false
    Let Infinitely Lived Particles Die When Emitter Deactivates=true
    Loop Particles Lifetime=true
   )
   &SolveForcesAndVelocity(Clamp Velocity=false, Limit Acceleration=false)
   &SkeletalMeshLocation(MaskSpawnGroup=false, Mesh Sampling Type=::NewEnumerator5, Skeletal Mesh=$User.Skeletal Mesh, Socket Sampling Mode=::NewEnumerator5)
   &DynamicMaterialParameters(
    Index 0 Param 1=20
    Module.Param0WriteEnabled=true
    Module.Param1WriteEnabled=true
    Module.Param2WriteEnabled=true
    Module.Param3WriteEnabled=true
   )
   &ScaleColor(
    Scale Alpha=&FN_FrontEndFloatModifier(Gameplay=0.7, Front End=0.2)
    Scale RGB=&MakeVectorFromLinearColorRGB(Color=&ColorFromCurve(
      ColorCurve=DataInterface() {propertyValues {
        AlphaCurve {keys [<at 0 1 RCIM_Constant>]}
        BlueCurve {keys [<at 0 2>,<at 0.4 0.0411 RCIM_Constant>]}
        GreenCurve {keys [<at 0 2>,<at 0.4 0.0312 RCIM_Constant>]}
        RedCurve {keys [<at 0 0.2549>,<at 0.4 1 RCIM_Constant>]}
       }}
      CurveIndex=$User.Damaged
     ))
    ScaleA=true
    ScaleRGB=true
   )
   &FN_Light_Attributes(
    Color=$Particles.Color
    Create Color Attribute=true
    Create Exponent Attribute=true
    Create Radius Attribute=true
    Exponent=3
    Radius=25
    Scale Color=&FloatFromCurve(
     FloatCurve=DataInterface() {propertyValues {Curve {keys [
         <at 0 1 RCIM_Cubic>
         {arriveTangent=0.2056, interpMode=RCIM_Cubic, leaveTangent=0.2056, time=0.0669, value=0.9233}
         {arriveTangent=-0.0227, interpMode=RCIM_Cubic, leaveTangent=-0.0227, tangentMode=RCTM_User, time=0.1294, value=1.0266}
         {arriveTangent=-1.0078, interpMode=RCIM_Cubic, leaveTangent=-1.0078, time=0.1628, value=0.8775}
         {arriveTangent=0.5087, interpMode=RCIM_Cubic, leaveTangent=0.5087, time=0.2007, value=0.9548}
         {arriveTangent=0.6543, interpMode=RCIM_Cubic, leaveTangent=0.6543, time=0.2486, value=0.9211}
         {arriveTangent=-0.5581, interpMode=RCIM_Cubic, leaveTangent=-0.5581, time=0.2893, value=1.0128}
         {arriveTangent=-1.1151, interpMode=RCIM_Cubic, leaveTangent=-1.1151, time=0.3196, value=0.8814}
         {arriveTangent=0.0233, interpMode=RCIM_Cubic, leaveTangent=0.0233, time=0.3426, value=0.9534}
         {arriveTangent=0.2188, interpMode=RCIM_Cubic, leaveTangent=0.2188, time=0.3842, value=0.8829}
         {arriveTangent=0.5503, interpMode=RCIM_Cubic, leaveTangent=0.5503, time=0.4256, value=0.9715}
         {arriveTangent=1.1505, interpMode=RCIM_Cubic, leaveTangent=1.1505, time=0.4783, value=0.9347}
         {arriveTangent=-0.543, interpMode=RCIM_Cubic, leaveTangent=-0.543, time=0.5208, value=1.081}
         {arriveTangent=-1.8762, interpMode=RCIM_Cubic, leaveTangent=-1.8762, time=0.5515, value=0.8949}
         {arriveTangent=-0.3447, interpMode=RCIM_Cubic, leaveTangent=-0.3447, time=0.5969, value=0.9382}
         {arriveTangent=-0.2654, interpMode=RCIM_Cubic, leaveTangent=-0.2654, time=0.6386, value=0.8649}
         {arriveTangent=0.0785, interpMode=RCIM_Cubic, leaveTangent=0.0785, time=0.6788, value=0.9165}
         {arriveTangent=0.5202, interpMode=RCIM_Cubic, leaveTangent=0.5202, time=0.7211, value=0.8713}
         {arriveTangent=0.2753, interpMode=RCIM_Cubic, leaveTangent=0.2753, time=0.7516, value=0.9544}
         {arriveTangent=0.2662, interpMode=RCIM_Cubic, leaveTangent=0.2662, time=0.783, value=0.8884}
         {arriveTangent=0.8297, interpMode=RCIM_Cubic, leaveTangent=0.8297, time=0.8244, value=0.9737}
         {arriveTangent=0.506, interpMode=RCIM_Cubic, leaveTangent=0.506, time=0.838, value=0.934}
         {arriveTangent=0.5794, interpMode=RCIM_Cubic, leaveTangent=0.5794, time=0.9054, value=1.0147}
         {arriveTangent=0.3362, interpMode=RCIM_Cubic, leaveTangent=0.3362, time=0.9341, value=0.9897}
         {arriveTangent=0.1563, interpMode=RCIM_Cubic, leaveTangent=0.1563, time=0.9623, value=1.0338}
         {interpMode=RCIM_Cubic, leaveTangent=0.0463, time=1, value=1}
        ]}}}
     Scale Curve=2
    )
   )
   &AlignSpriteToMeshOrientation(
    Mesh Orientation Relative Sprite Alignment Vector=<1 0 1>
    Orientation Quaternion:Quat4f=$Particles.SkeletalMeshLocation.SampledRotation
    SetAlignment=true
    SetFacing=true
   )
  ]
  rendererClasses [&NiagaraSpriteRendererProperties, &NiagaraLightRendererProperties]
  renderers [
   {rendererClass=&NiagaraSpriteRendererProperties}
   {rendererClass=&NiagaraLightRendererProperties, rendererIndex=1}
  ]
 }
 ShipEndGlows {
  emitterUpdateScript [
   &EmitterState()
   &SpawnBurst_Instantaneous(Spawn Count=4, Use Spawn Probability=false, UseLoopCountLimit=true)
  ]
  particleSpawnScript [
   &PartitionParticles(Number Of Partitions=4)
   &InitializeParticle(
    Color=<rgba 1 1 1 0.03>
    Lifetime=2
    Sprite Size=&SelectVector2DFromArray(
     Direct Array Index=$Particles.PartitionParticles.Partition
     Vector 2D Selection Array=DataInterface() {propertyValues {FloatData [<26 26>, <26 26>, <16 16>, <16 16>]}}
    )
    Sprite Size Mode=::NewEnumerator4
    UsePositionOffset=false
    Write Material Random=true
   )
   &CameraOffset(Camera Offset Amount=-2)
  ]
  particleUpdateScript [
   &ParticleState(
    Kill Particles When Lifetime Has Elapsed=false
    Let Infinitely Lived Particles Die When Emitter Deactivates=true
    Loop Particles Lifetime=true
   )
   &SkeletalMeshLocation(
    Bone / Socket Index=1
    MaskSpawnGroup=false
    Mesh Sampling Type=::NewEnumerator5
    Sampled Position Offset=&SelectVectorFromArray(
     Direct Array Index=$Particles.PartitionParticles.Partition
     Vector Selection Array=DataInterface() {propertyValues {FloatData [<5 6.5 0>, <5 -6.5 0>, <-0.5 3.25 0>, <-0.5 -3.25 0>]}}
    )
    Skeletal Mesh=$User.Skeletal Mesh
    Socket Sampling Mode=::NewEnumerator5
   )
   &ScaleSpriteSize(Uniform Curve Sprite Scale=DataInterface() {propertyValues {Curve {keys [
        {arriveTangent=2.3436, interpMode=RCIM_Cubic, time=0, value=1}
        {arriveTangent=0.0789, interpMode=RCIM_Cubic, leaveTangent=0.0789, time=0.0802, value=0.9722}
        {arriveTangent=-0.0227, interpMode=RCIM_Cubic, leaveTangent=-0.0227, tangentMode=RCTM_User, time=0.1732, value=1.0137}
        {arriveTangent=-0.2717, interpMode=RCIM_Cubic, leaveTangent=-0.2717, time=0.2228, value=0.9539}
        {arriveTangent=0.1371, interpMode=RCIM_Cubic, leaveTangent=0.1371, time=0.2792, value=0.9849}
        {arriveTangent=0.1764, interpMode=RCIM_Cubic, leaveTangent=0.1764, time=0.3503, value=0.9714}
        {arriveTangent=-0.1504, interpMode=RCIM_Cubic, leaveTangent=-0.1504, time=0.411, value=1.0081}
        {arriveTangent=-0.3006, interpMode=RCIM_Cubic, leaveTangent=-0.3006, time=0.456, value=0.9555}
        {arriveTangent=0.0063, interpMode=RCIM_Cubic, leaveTangent=0.0063, time=0.4903, value=0.9843}
        {arriveTangent=0.059, interpMode=RCIM_Cubic, leaveTangent=0.059, time=0.5521, value=0.9561}
        {arriveTangent=0.1483, interpMode=RCIM_Cubic, leaveTangent=0.1483, time=0.6136, value=0.9916}
        {arriveTangent=0.3101, interpMode=RCIM_Cubic, leaveTangent=0.3101, time=0.6919, value=0.9768}
        {arriveTangent=-0.1464, interpMode=RCIM_Cubic, leaveTangent=-0.1464, time=0.7551, value=1.0355}
        {arriveTangent=-0.5058, interpMode=RCIM_Cubic, leaveTangent=-0.5058, time=0.8008, value=0.9609}
        {arriveTangent=-0.0929, interpMode=RCIM_Cubic, leaveTangent=-0.0929, time=0.8683, value=0.9782}
        {arriveTangent=0.1654, interpMode=RCIM_Cubic, leaveTangent=0.1654, time=0.9304, value=0.9488}
        {interpMode=RCIM_Cubic, leaveTangent=0.0463, time=1, value=1}
       ]}}})
   &DynamicMaterialParameters(
    Index 0 Param 1=50
    Module.Param0WriteEnabled=true
    Module.Param1WriteEnabled=true
    Module.Param2WriteEnabled=true
    Module.Param3WriteEnabled=true
   )
   &ScaleColor(
    Scale Alpha=&FN_FrontEndFloatModifier(Gameplay=0.35, Front End=0.3)
    Scale RGB=&MakeVectorFromLinearColorRGB(Color=&ColorFromCurve(
      ColorCurve=DataInterface() {propertyValues {
        AlphaCurve {keys [<at 0 1 RCIM_Constant>]}
        BlueCurve {keys [<at 0.05 2>,<at 0.5 0.0411 RCIM_Constant>]}
        GreenCurve {keys [<at 0.05 2>,<at 0.5 0.0312 RCIM_Constant>]}
        RedCurve {keys [<at 0.05 0.2549>,<at 0.5 1 RCIM_Constant>]}
       }}
      CurveIndex=$User.Damaged
     ))
    ScaleA=true
    ScaleRGB=true
   )
   &FN_Light_Attributes(
    Color=$Particles.Color
    Create Color Attribute=true
    Create Exponent Attribute=true
    Create Position Attribute=true
    Create Radius Attribute=true
    Enable Particle Light Enable=true
    Exponent=3
    Light Enabled:NiagaraBool=&SetBoolByIntComparison(A=$Particles.PartitionParticles.Partition, B=1, Comparison Type=::NewEnumerator7)
    Position=&Add_Vector(
     A=$Output.SkeletalMeshLocation.SampledPositionWithoutOffset
     B=&RotateVectorByQuaternion(
      Vector To Rotate=&Lerp_Vector(A=<2 6.5 0>, B=<2 -6.5 0>, Alpha=&MakeFloatFromInt(Int=$Particles.PartitionParticles.Partition))
      Quaternion=$Particles.SkeletalMeshLocation.SampledRotation
     )
    )
    Radius=20
    Scale Color=&FloatFromCurve(
     FloatCurve=DataInterface() {propertyValues {Curve {keys [
         <at 0 1 RCIM_Cubic>
         {arriveTangent=0.2056, interpMode=RCIM_Cubic, leaveTangent=0.2056, time=0.0669, value=0.9233}
         {arriveTangent=-0.0227, interpMode=RCIM_Cubic, leaveTangent=-0.0227, tangentMode=RCTM_User, time=0.1294, value=1.0266}
         {arriveTangent=-1.0078, interpMode=RCIM_Cubic, leaveTangent=-1.0078, time=0.1628, value=0.8775}
         {arriveTangent=0.5087, interpMode=RCIM_Cubic, leaveTangent=0.5087, time=0.2007, value=0.9548}
         {arriveTangent=0.6543, interpMode=RCIM_Cubic, leaveTangent=0.6543, time=0.2486, value=0.9211}
         {arriveTangent=-0.5581, interpMode=RCIM_Cubic, leaveTangent=-0.5581, time=0.2893, value=1.0128}
         {arriveTangent=-1.1151, interpMode=RCIM_Cubic, leaveTangent=-1.1151, time=0.3196, value=0.8814}
         {arriveTangent=0.0233, interpMode=RCIM_Cubic, leaveTangent=0.0233, time=0.3426, value=0.9534}
         {arriveTangent=0.2188, interpMode=RCIM_Cubic, leaveTangent=0.2188, time=0.3842, value=0.8829}
         {arriveTangent=0.5503, interpMode=RCIM_Cubic, leaveTangent=0.5503, time=0.4256, value=0.9715}
         {arriveTangent=1.1505, interpMode=RCIM_Cubic, leaveTangent=1.1505, time=0.4783, value=0.9347}
         {arriveTangent=-0.543, interpMode=RCIM_Cubic, leaveTangent=-0.543, time=0.5208, value=1.081}
         {arriveTangent=-1.8762, interpMode=RCIM_Cubic, leaveTangent=-1.8762, time=0.5515, value=0.8949}
         {arriveTangent=-0.3447, interpMode=RCIM_Cubic, leaveTangent=-0.3447, time=0.5969, value=0.9382}
         {arriveTangent=-0.2654, interpMode=RCIM_Cubic, leaveTangent=-0.2654, time=0.6387, value=0.8649}
         {arriveTangent=0.0785, interpMode=RCIM_Cubic, leaveTangent=0.0785, time=0.6788, value=0.9165}
         {arriveTangent=0.5202, interpMode=RCIM_Cubic, leaveTangent=0.5202, time=0.7211, value=0.8713}
         {arriveTangent=0.2753, interpMode=RCIM_Cubic, leaveTangent=0.2753, time=0.7516, value=0.9544}
         {arriveTangent=0.2662, interpMode=RCIM_Cubic, leaveTangent=0.2662, time=0.783, value=0.8884}
         {arriveTangent=0.8297, interpMode=RCIM_Cubic, leaveTangent=0.8297, time=0.8244, value=0.9737}
         {arriveTangent=0.506, interpMode=RCIM_Cubic, leaveTangent=0.506, time=0.838, value=0.934}
         {arriveTangent=0.5794, interpMode=RCIM_Cubic, leaveTangent=0.5794, time=0.9054, value=1.0147}
         {arriveTangent=0.3362, interpMode=RCIM_Cubic, leaveTangent=0.3362, time=0.9341, value=0.9897}
         {arriveTangent=0.1563, interpMode=RCIM_Cubic, leaveTangent=0.1563, time=0.9623, value=1.0338}
         {interpMode=RCIM_Cubic, leaveTangent=0.0463, time=1, value=1}
        ]}}}
     Scale Curve=2
    )
   )
  ]
  rendererClasses [&NiagaraSpriteRendererProperties, &NiagaraLightRendererProperties]
  renderers [
   {rendererClass=&NiagaraSpriteRendererProperties}
   {rendererClass=&NiagaraLightRendererProperties, rendererIndex=1}
  ]
 }
}
summary {
 systemName=NS_Backpack_CrimsonPeak_Ship
 emitters [
  {emitterName=ShipBodyGlow, rendererClasses [&NiagaraSpriteRendererProperties, &NiagaraLightRendererProperties]}
  {emitterName=ShipEndGlows, rendererClasses [&NiagaraSpriteRendererProperties, &NiagaraLightRendererProperties]}
 ]
 userVariables [
  User.bIsFrontEnd {type=NiagaraBool}
  User.bIsFrontEndPreview {type=NiagaraBool}
  User.Damaged {type=NiagaraFloat, defaultValue=0, description=""}
  User.Skeletal Mesh {
   type=NiagaraDataInterfaceSkeletalMesh
   defaultValue {
    struct=NiagaraExt_VariableValue_DataInterface
    value {dataInterface=&self:NiagaraDataInterfaceSkeletalMesh_0}
   }
  }
 ]
}
