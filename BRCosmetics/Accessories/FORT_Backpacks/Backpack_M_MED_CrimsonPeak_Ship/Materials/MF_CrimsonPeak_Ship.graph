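# Material function for the CrimsonPeak ship back bling that overrides only the emissive channel of incoming material attributes to show a damaged state.
# IsDamagedFromBP, apparently Blueprint-driven, blends the original emissive toward a damage color scaled by an intensity and the emissive mask's red channel.
# A time-driven sine, passed through Frac and clamped to a floor set by Pulse Strength, pulses the blended result, including the undamaged glow.
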MaterialFunction /BRCosmetics/Accessories/FORT_Backpacks/Backpack_M_MED_CrimsonPeak_Ship/Materials/MF_CrimsonPeak_Ship
graph {nodes [
  n1 = GetMaterialAttributes(None=FunctionInput() {inputType=FunctionInput_MaterialAttributes}) {attributeGetTypes [#1]}
  FunctionOutput(None=SetMaterialAttributes!(MaterialAttributes=@n1.MaterialAttributes, EmissiveColor=NamedRerouteUsage()) {attributeSetTypes [#1]})
  NamedRerouteDeclaration(None=@n1.EmissiveColor) {name=inEmissive, nodeColor=<rgba 1 0.7734 0 1>}
  n2 = NamedRerouteUsage()
  n3 = TextureSampleParameter2D() {
   group=CrimsonPeak
   parameterName=Emissive
   samplerType=SAMPLERTYPE_Masks
   texture=/BRCosmetics/Accessories/FORT_Backpacks/Backpack_M_MED_CrimsonPeak_Ship/Textures/T_F_MED_CrimsonPeak_Ship_Backpack_E
  }
  n4 = Multiply(
   A=VectorParameter() {parameterName=ShipDamageColor, defaultValue=<rgba 1 0 0 1>}.RGB
   B=Multiply(A=@n3.R, B=ScalarParameter() {defaultValue=5, group=CrimsonPeak, parameterName=EmissiveIntenisty}) {constB=20}
  )
  n5 = ScalarParameter() {group=CrimsonPeak, parameterName=IsDamagedFromBP}
  n6 = ScalarParameter() {defaultValue=0.8, group=CrimsonPeak, parameterName=Pulse Speed}
  n7 = Time()
  n8 = Frac(None=Sine(None=Multiply(A=@n6, B=@n7)))
  n9 = OneMinus(None=ScalarParameter() {defaultValue=0.5, group=CrimsonPeak, parameterName=Pulse Strength})
  NamedRerouteDeclaration(None=Multiply(A=LinearInterpolate(A=@n2, B=@n4, Alpha=@n5), B=Clamp(None=@n8, Min=@n9) {minDefault=0.25})) {name=outEmissive, nodeColor=<rgba 0.1719 0 1 1>}
 ]}
